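# Physics asset for the Esmeralda character's FaceAcc_BackBling mesh part, set up for secondary motion on her hair.
# Capsules on the head, neck, spine, clavicles and upper arms provide body volumes around the hair.
# Back, left-front and left-side chains hang from head-linked base bodies with narrow swing and mostly zero twist limits; right-front pieces branch off the right back chain.
# Free-angle cross constraints tie neighboring strands together.
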PhysicsAsset /BRCosmetics/Characters/Player/Female/Medium/Bodies/F_MED_Esmeralda/Meshes/Parts/F_MED_Esmeralda_FaceAcc_BackBling_Physics
bodies {
 base_BackHair_L [{shapeType=Capsule, length=0, radius=0.5, shapeName=base_BackHair_L_capsule, rotation=<rot 90 0 0>}]
 base_BackHair_M [{shapeType=Capsule, length=0, radius=0.5, shapeName=base_BackHair_M_capsule, rotation=<rot 90 0 0>}]
 base_BackHair_R [{shapeType=Capsule, length=0, radius=0.5, shapeName=base_BackHair_R_capsule, rotation=<rot 90 0 0>}]
 base_FrontHair_L [{shapeType=Capsule, length=0, radius=0.5, shapeName=base_FrontHair_L_capsule, rotation=<rot 90 0 0>}]
 base_SideHairA_L [{shapeType=Capsule, length=0, radius=0.5, shapeName=base_SideHairA_L_capsule, rotation=<rot 90 0 0>}]
 clavicle_l [{
   shapeType=Capsule
   length=9
   radius=6.25
   shapeName=clavicle_r_capsule
   center=<4.8507 -0.65 -1.3777>
   rotation=<rot 76.4723 -0.0009 179.999>
  }]
 clavicle_r [
  {
   shapeType=Capsule
   length=8.1499
   radius=5.8981
   shapeName=clavicle_r_capsule
   center=<-4.5151 0.65 1.0322>
   rotation=<rot 75.8544 -0.002 -0.0021>
  }
  {
   shapeType=Capsule
   length=2
   radius=3
   shapeName=clavicle_r_capsule
   center=<-5.4397 3.8432 -2.1637>
   rotation=<rot -13.6323 -5.2919 -135.9604>
  }
  {
   shapeType=Capsule
   length=6
   radius=3
   shapeName=clavicle_r_capsule
   center=<-5.6831 -1.3816 -2.0707>
   rotation=<rot -1.3127 3.5701 111.5968>
  }
 ]
 dyn_BackHairPiece_R_01 [
  {
   shapeType=Capsule
   length=1.1263
   radius=1
   shapeName=dyn_BackHairPiece_R_01_capsule
   center=<1.6273 -0.302 -1.2205>
   rotation=<rot 11.2842 0 0>
  }
  {
   shapeType=Capsule
   length=1.8984
   radius=0.75
   shapeName=dyn_BackHair_R_03_capsule
   center=<-0.2398 0.0181 0.038>
   rotation=<rot 75.1808 2.5767 6.2374>
  }
 ]
 dyn_BackHair_L_01 [{shapeType=Capsule, length=2, radius=0.77, shapeName=dyn_BackHair_L_01_capsule, center=<1.5 0 0.5>, rotation=<rot 90 0 0>}]
 dyn_BackHair_L_02 [{shapeType=Capsule, length=2, radius=1.5, shapeName=dyn_BackHair_L_02_capsule, center=<2 0 0>, rotation=<rot 90 0 0>}]
 dyn_BackHair_L_03 [
  {shapeType=Capsule, length=1.5, radius=2, shapeName=dyn_BackHair_L_03_capsule, center=<2 0 0>, rotation=<rot 90 0 0>}
  {
   shapeType=Capsule
   length=3
   radius=2
   shapeName=dyn_BackHair_L_03_capsule
   center=<1.4431 1.3595 -1.524>
   rotation=<rot 62.1093 180 -167.3966>
  }
 ]
 dyn_BackHair_L_04 [{
   shapeType=Capsule
   length=1.5
   radius=1.5
   shapeName=dyn_BackHair_L_04_capsule
   center=<1.5146 0.2079 1.6956>
   rotation=<rot 37.737 180 -137.8983>
  }]
 dyn_BackHair_M_01 [{shapeType=Capsule, length=2, radius=1, shapeName=dyn_BackHair_M_01_capsule, center=<1.5 0 0.5>, rotation=<rot 90 0 0>}]
 dyn_BackHair_M_02 [{shapeType=Capsule, length=2, radius=1.5, shapeName=dyn_BackHair_M_02_capsule, center=<1.5 -0.2586 0.2491>, rotation=<rot 90 0 0>}]
 dyn_BackHair_M_03 [
  {shapeType=Capsule, length=1, radius=2, shapeName=dyn_BackHair_M_03_capsule, center=<1.5 -0.453 -0.1192>, rotation=<rot 90 0 0>}
  {shapeType=Capsule, length=1, radius=2, shapeName=dyn_BackHair_M_03_capsule, center=<0.9676 2.5824 -1.302>, rotation=<rot 90 0 0>}
 ]
 dyn_BackHair_M_04 [{shapeType=Capsule, length=1.5, radius=2, shapeName=dyn_BackHair_M_04_capsule, center=<1.5 -0.4874 -0.2733>, rotation=<rot 90 0 0>}]
 dyn_BackHair_M_05 [
  {
   shapeType=Capsule
   length=4.3524
   radius=2.25
   shapeName=dyn_BackHair_M_05_capsule
   center=<4.6593 -1.1839 0.7622>
   rotation=<rot 82.9594 -6.8427 0.0852>
  }
  {
   shapeType=Capsule
   length=0.5059
   radius=1.5
   shapeName=dyn_BackHair_M_05_capsule
   center=<2.1132 -3.583 0.068>
   rotation=<rot 74.3653 169.7266 -177.3021>
  }
  {
   shapeType=Capsule
   length=1.2982
   radius=1.5378
   shapeName=dyn_BackHair_M_05_capsule
   center=<1.943 2.7089 -1.1753>
   rotation=<rot 76.7283 180 -180>
  }
 ]
 dyn_BackHair_R_01 [{shapeType=Capsule, length=0.5, radius=0.5, shapeName=dyn_BackHair_R_01_capsule, center=<1.5 0 0.5>, rotation=<rot 90 0 0>}]
 dyn_BackHair_R_02 [
  {shapeType=Capsule, length=2, radius=1.5, shapeName=dyn_BackHair_R_02_capsule, center=<1.5 0 0.2073>, rotation=<rot 90 0 0>}
  {
   shapeType=Capsule
   length=3.5
   radius=1.25
   shapeName=dyn_BackHair_R_02_capsule
   center=<2.0353 -1.593 -0.8823>
   rotation=<rot 82.2741 0.0001 17.0478>
  }
 ]
 dyn_BackHair_R_03 [{shapeType=Capsule, length=1.3106, radius=1.9394, shapeName=dyn_BackHair_R_03_capsule, center=<1.5 0 -0.2635>, rotation=<rot 90 0 0>}]
 dyn_BackHair_R_04 [
  {
   shapeType=Capsule
   length=3
   radius=1.75
   shapeName=dyn_BackHair_R_04_capsule
   center=<3.348 0.1557 0.8001>
   rotation=<rot 70.2908 179.9999 177.5467>
  }
  {
   shapeType=Capsule
   length=1.5
   radius=1
   shapeName=dyn_BackHair_R_04_capsule
   center=<2.3783 -1.9716 -0.5407>
   rotation=<rot 48.3381 179.9988 179.9997>
  }
 ]
 dyn_FrontHairA_R_01 [{
   shapeType=Capsule
   length=2.6692
   radius=0.7302
   shapeName=dyn_FrontHairA_R_01_capsule
   center=<1.9247 -0.3776 0.408>
   rotation=<rot 0 77.5728 -73.9372>
  }]
 dyn_FrontHairB_R_01 [{
   shapeType=Capsule
   length=3.1499
   radius=1.1524
   shapeName=dyn_FrontHairB_R_01_capsule
   center=<2.1371 -0.8454 0.045>
   rotation=<rot 0 61.0407 -85.3097>
  }]
 dyn_FrontHair_L_01 [{shapeType=Capsule, length=2, radius=1, shapeName=dyn_FrontHair_L_01_capsule, center=<2 0 0>, rotation=<rot 90 0 0>}]
 dyn_FrontHair_L_02 [{shapeType=Capsule, length=2, radius=1.5, shapeName=dyn_FrontHair_L_02_capsule, center=<2.4715 0 0>, rotation=<rot 90 0 0>}]
 dyn_FrontHair_L_03 [{shapeType=Capsule, length=2, radius=1.5, shapeName=dyn_FrontHair_L_03_capsule, center=<2 0 0>, rotation=<rot 90 0 0>}]
 dyn_FrontHair_L_04 [
  {shapeType=Capsule, length=5, radius=1.75, shapeName=dyn_FrontHair_L_04_capsule, center=<5.0979 0 -0.6245>, rotation=<rot 79.4339 0.0001 0>}
  {
   shapeType=Capsule
   length=2.2204
   radius=1.5
   shapeName=dyn_FrontHair_L_04_capsule
   center=<5.4232 -2.2405 -1.0116>
   rotation=<rot 79.4339 0 -8.7407>
  }
 ]
 dyn_SideHairA_L_01 [{shapeType=Capsule, length=2, radius=1.5, shapeName=dyn_SideHairA_L_01_capsule, center=<2 0 0>, rotation=<rot 90 0 0>}]
 dyn_SideHairA_L_02 [{shapeType=Capsule, length=2, radius=1.5, shapeName=dyn_SideHairA_L_02_capsule, center=<2 0 0>, rotation=<rot 90 0 0>}]
 dyn_SideHairA_L_03 [{
   shapeType=Capsule
   length=1.889
   radius=2.111
   shapeName=dyn_SideHairA_L_03_capsule
   center=<3.2248 -1.1237 0.0567>
   rotation=<rot 90 -153.4349 -132.6269>
  }]
 head [
  {shapeType=Capsule, length=9.5, radius=6.6, shapeName=head_capsule, center=<-0.0952 1.5803 0>, rotation=<rot 0 23 90>}
  {shapeType=Capsule, length=5, radius=6, shapeName=head_capsule, center=<5.0043 1.4375 0.5566>, rotation=<rot 0 0 90>}
 ]
 neck_01 [{shapeType=Capsule, length=5.5, radius=6, shapeName=neck_01_capsule, center=<0.9625 1.5979 0>, rotation=<rot 90 180 -178.1933>}]
 neck_02 [{shapeType=Capsule, length=1, radius=6, shapeName=neck_02_capsule, center=<-1.0922 1.0064 0>}]
 spine_03 [{shapeType=Capsule, length=7.5049, radius=11.4951, shapeName=spine_03_capsule, center=<8.6049 0.7707 0>, rotation=<rot 90 178.2101 -175.6809>}]
 spine_05 [
  {shapeType=Capsule, length=17.5, radius=8.5, shapeName=spine_05_capsule, center=<1.8557 -0.3374 0>, rotation=<rot 0 -90 0>}
  {shapeType=Capsule, length=2.5, radius=6, shapeName=spine_05_capsule, center=<8.819 1.2937 0.0998>}
  {shapeType=Capsule, length=6, radius=3, shapeName=spine_05_capsule, center=<5.4804 -6.0841 5.2242>, rotation=<rot -45.895 -0.0009 21.556>}
  {shapeType=Capsule, length=30.5, radius=15, shapeName=spine_05_capsule, center=<23.9017 -25.9446 0>}
  {shapeType=Capsule, length=15, radius=15, shapeName=spine_05_capsule, center=<7.1119 -30.5702 0>}
 ]
 upperarm_l [
  {
   shapeType=Capsule
   length=21.4
   radius=6.9
   shapeName=upperarm_l_capsule
   center=<15.6638 0 -0.8858>
   rotation=<rot 84.7743 -179.9942 -179.9946>
  }
  {shapeType=Capsule, length=0, radius=6, shapeName=upperarm_l_capsule, center=<0 0 -1.1886>}
 ]
 upperarm_r [{
   shapeType=Capsule
   length=21.4
   radius=6.9
   shapeName=upperarm_l_capsule
   center=<-13.0462 0.5546 1.1252>
   rotation=<rot 84.7743 -179.9942 0.0054>
  }]
}
constraints [
 {bone1Name=dyn_BackHair_L_02, bone2Name=dyn_BackHair_L_01, swing1LimitDegrees=20, swing2LimitDegrees=10, twistLimitDegrees=0}
 {bone1Name=dyn_BackHair_L_03, bone2Name=dyn_BackHair_L_02, swing1LimitDegrees=15, swing2LimitDegrees=20, twistLimitDegrees=0}
 {bone1Name=dyn_BackHair_L_04, bone2Name=dyn_BackHair_L_03, swing1LimitDegrees=20, swing2LimitDegrees=20, twistLimitDegrees=0}
 {bone1Name=dyn_BackHair_M_02, bone2Name=dyn_BackHair_M_01, swing1LimitDegrees=20, swing2LimitDegrees=10, twistLimitDegrees=0}
 {bone1Name=dyn_BackHair_M_03, bone2Name=dyn_BackHair_M_02, swing1LimitDegrees=20, swing2LimitDegrees=20, twistLimitDegrees=0}
 {bone1Name=dyn_BackHair_M_04, bone2Name=dyn_BackHair_M_03, swing1LimitDegrees=20, swing2LimitDegrees=20, twistLimitDegrees=0}
 {bone1Name=dyn_BackHair_M_05, bone2Name=dyn_BackHair_M_04, swing1LimitDegrees=20, swing2LimitDegrees=20, twistLimitDegrees=0}
 {bone1Name=dyn_BackHair_R_02, bone2Name=dyn_BackHair_R_01, swing1LimitDegrees=20, swing2LimitDegrees=20, twistLimitDegrees=10}
 {bone1Name=dyn_BackHair_R_03, bone2Name=dyn_BackHair_R_02, swing1LimitDegrees=5, swing2LimitDegrees=25, twistLimitDegrees=0}
 {bone1Name=dyn_BackHair_R_04, bone2Name=dyn_BackHair_R_03, swing1LimitDegrees=5, swing2LimitDegrees=35, twistLimitDegrees=0}
 {bone1Name=dyn_FrontHair_L_02, bone2Name=dyn_FrontHair_L_01, swing1LimitDegrees=10, swing2LimitDegrees=20, twistLimitDegrees=0}
 {bone1Name=dyn_FrontHair_L_03, bone2Name=dyn_FrontHair_L_02, swing1LimitDegrees=15, swing2LimitDegrees=20, twistLimitDegrees=0}
 {bone1Name=dyn_FrontHair_L_04, bone2Name=dyn_FrontHair_L_03, swing1LimitDegrees=15, swing2LimitDegrees=20, twistLimitDegrees=0}
 {bone1Name=dyn_SideHairA_L_02, bone2Name=dyn_SideHairA_L_01, swing1LimitDegrees=10, swing2LimitDegrees=20, twistLimitDegrees=0}
 {bone1Name=dyn_SideHairA_L_03, bone2Name=dyn_SideHairA_L_02, swing1LimitDegrees=10, swing2LimitDegrees=20, twistLimitDegrees=0}
 {bone1Name=dyn_FrontHair_L_01, bone2Name=base_FrontHair_L, swing1LimitDegrees=7.5, swing2LimitDegrees=10, twistLimitDegrees=0}
 {bone1Name=dyn_SideHairA_L_01, bone2Name=base_SideHairA_L, swing1LimitDegrees=10, swing2LimitDegrees=15, twistLimitDegrees=0}
 {bone1Name=dyn_BackHair_L_01, bone2Name=base_BackHair_L, swing1LimitDegrees=10, swing2LimitDegrees=15, twistLimitDegrees=0}
 {bone1Name=dyn_BackHair_M_01, bone2Name=base_BackHair_M, swing1LimitDegrees=10, swing2LimitDegrees=15, twistLimitDegrees=0}
 {bone1Name=dyn_BackHair_R_01, bone2Name=base_BackHair_R, swing1LimitDegrees=10, swing2LimitDegrees=20, twistLimitDegrees=0}
 {
  bone1Name=head
  bone2Name=neck_02
  swing1LimitDegrees=45
  swing1Motion=Free
  swing2LimitDegrees=45
  swing2Motion=Free
  twistLimitDegrees=45
  twistMotion=Free
 }
 {
  bone1Name=neck_02
  bone2Name=neck_01
  swing1LimitDegrees=45
  swing1Motion=Free
  swing2LimitDegrees=45
  swing2Motion=Free
  twistLimitDegrees=45
  twistMotion=Free
 }
 {
  bone1Name=neck_01
  bone2Name=spine_05
  swing1LimitDegrees=45
  swing1Motion=Free
  swing2LimitDegrees=45
  swing2Motion=Free
  twistLimitDegrees=45
  twistMotion=Free
 }
 {
  bone1Name=upperarm_l
  bone2Name=clavicle_l
  swing1LimitDegrees=45
  swing1Motion=Free
  swing2LimitDegrees=45
  swing2Motion=Free
  twistLimitDegrees=45
  twistMotion=Free
 }
 {
  bone1Name=base_BackHair_L
  bone2Name=head
  swing1LimitDegrees=45
  swing1Motion=Free
  swing2LimitDegrees=45
  swing2Motion=Free
  twistLimitDegrees=45
  twistMotion=Free
 }
 {
  bone1Name=base_BackHair_M
  bone2Name=head
  swing1LimitDegrees=45
  swing1Motion=Free
  swing2LimitDegrees=45
  swing2Motion=Free
  twistLimitDegrees=45
  twistMotion=Free
 }
 {
  bone1Name=base_BackHair_R
  bone2Name=head
  swing1LimitDegrees=45
  swing1Motion=Free
  swing2LimitDegrees=45
  swing2Motion=Free
  twistLimitDegrees=45
  twistMotion=Free
 }
 {
  bone1Name=base_FrontHair_L
  bone2Name=head
  swing1LimitDegrees=45
  swing1Motion=Free
  swing2LimitDegrees=45
  swing2Motion=Free
  twistLimitDegrees=45
  twistMotion=Free
 }
 {
  bone1Name=base_SideHairA_L
  bone2Name=head
  swing1LimitDegrees=45
  swing1Motion=Free
  swing2LimitDegrees=45
  swing2Motion=Free
  twistLimitDegrees=45
  twistMotion=Free
 }
 {
  bone1Name=clavicle_l
  bone2Name=spine_05
  swing1LimitDegrees=45
  swing1Motion=Free
  swing2LimitDegrees=45
  swing2Motion=Free
  twistLimitDegrees=45
  twistMotion=Free
 }
 {
  bone1Name=clavicle_r
  bone2Name=spine_05
  swing1LimitDegrees=45
  swing1Motion=Free
  swing2LimitDegrees=45
  swing2Motion=Free
  twistLimitDegrees=45
  twistMotion=Free
 }
 {
  bone1Name=dyn_BackHair_M_02
  bone2Name=dyn_BackHair_R_02
  swing1LimitDegrees=45
  swing1Motion=Free
  swing2LimitDegrees=45
  swing2Motion=Free
  twistLimitDegrees=45
  twistMotion=Free
 }
 {
  bone1Name=dyn_BackHair_M_02
  bone2Name=dyn_BackHair_L_02
  swing1LimitDegrees=45
  swing1Motion=Free
  swing2LimitDegrees=45
  swing2Motion=Free
  twistLimitDegrees=45
  twistMotion=Free
 }
 {
  bone1Name=dyn_BackHair_L_04
  bone2Name=dyn_BackHair_M_05
  swing1LimitDegrees=45
  swing1Motion=Free
  swing2LimitDegrees=45
  swing2Motion=Free
  twistLimitDegrees=45
  twistMotion=Free
 }
 {
  bone1Name=dyn_BackHair_R_04
  bone2Name=dyn_BackHair_M_05
  swing1LimitDegrees=45
  swing1Motion=Free
  swing2LimitDegrees=45
  swing2Motion=Free
  twistLimitDegrees=45
  twistMotion=Free
 }
 {bone1Name=dyn_BackHairPiece_R_01, bone2Name=dyn_BackHair_R_03, swing1LimitDegrees=15, swing2LimitDegrees=20, twistLimitDegrees=45}
 {
  bone1Name=spine_05
  bone2Name=spine_03
  swing1LimitDegrees=45
  swing1Motion=Free
  swing2LimitDegrees=45
  swing2Motion=Free
  twistLimitDegrees=45
  twistMotion=Free
 }
 {bone1Name=dyn_FrontHairB_R_01, bone2Name=dyn_BackHair_R_01, swing1LimitDegrees=15, swing2LimitDegrees=10, twistLimitDegrees=0}
 {bone1Name=dyn_FrontHairA_R_01, bone2Name=dyn_BackHair_R_01, swing1LimitDegrees=10, swing2LimitDegrees=0, twistLimitDegrees=0}
 {
  bone1Name=dyn_FrontHairB_R_01
  bone2Name=dyn_BackHair_R_02
  swing1LimitDegrees=45
  swing1Motion=Free
  swing2LimitDegrees=45
  swing2Motion=Free
  twistLimitDegrees=45
  twistMotion=Free
 }
 {
  bone1Name=upperarm_r
  bone2Name=clavicle_r
  swing1LimitDegrees=45
  swing1Motion=Free
  swing2LimitDegrees=45
  swing2Motion=Free
  twistLimitDegrees=45
  twistMotion=Free
 }
 {
  bone1Name=dyn_SideHairA_L_03
  bone2Name=dyn_FrontHair_L_03
  swing1LimitDegrees=45
  swing1Motion=Free
  swing2LimitDegrees=45
  swing2Motion=Free
  twistLimitDegrees=45
  twistMotion=Free
 }
 {
  bone1Name=dyn_BackHairPiece_R_01
  bone2Name=dyn_BackHair_R_03
  swing1LimitDegrees=45
  swing1Motion=Free
  swing2LimitDegrees=45
  swing2Motion=Free
  twistLimitDegrees=45
  twistMotion=Free
 }
]
