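# Animation Blueprint for Esmeralda's FaceAcc mesh, driving secondary motion of hair and hoop earrings.
# Body and face poses are copied and layered, then AnimDynamics simulates bangs and hoops.
# Additive bone offsets adjust hair using head pitch, aiming, skydive and backbling alphas computed each update.
# The cached result feeds two RigidBody simulations with gravity override: a backbling physics asset, or the default plus back-hair tweaks, selected by backbling state.
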AnimBlueprint /BRCosmetics/Characters/Player/Female/Medium/Bodies/F_MED_Esmeralda/Meshes/Parts/F_MED_Esmeralda_FaceAcc_AnimBP
target_skeleton=&/F_MED_Esmeralda_FaceAcc_Skeleton
graphs [
 AnimGraph {nodes [
   n1 = RigidBody!(
    ComponentPose=Local To Component(LocalPose=Use cached pose 'Logic2RigidBody'())
    Alpha
    OverrideWorldGravity:Vector=Get Gravity Override()
    ExternalForce
   ) -> (Pose) {
    bOverrideWorldGravity=true
    overridePhysicsAsset=&/F_MED_Esmeralda_FaceAcc_BackBling_Physics
    simulationSpace=BaseBoneSpace
    alphaBoolBlend {blendOption=Linear}
    baseBoneRef {boneName=spine_05}
    componentAppliedLinearAccClamp=<5000 5000 5000>
    componentLinearAccScale=<0.3 0.2>
    componentLinearVelScale=<0.3 0.2 0.2>
    simSpaceSettings {velocityScaleZ=0.5, worldAlpha=0.25, externalLinearDragV=<0.2 0.2 0.2>}
   }
   n2 = RigidBody(ComponentPose=Local To Component(LocalPose=Use cached pose 'Logic2RigidBody'()), OverrideWorldGravity=Get Gravity Override()) {
    bOverrideWorldGravity=true
    overridePhysicsAsset=&/F_MED_Esmeralda_FaceAcc_Physics
    simulationSpace=BaseBoneSpace
    alphaBoolBlend {blendOption=Linear}
    baseBoneRef {boneName=spine_05}
    componentAppliedLinearAccClamp=<5000 5000 5000>
    componentLinearAccScale=<0.3 0.2>
    componentLinearVelScale=<0.3 0.2 0.2>
    simSpaceSettings {velocityScaleZ=0.5, worldAlpha=0.25, externalLinearDragV=<0.2 0.2 0.2>}
   }
   n3 = "Transform (Modify) Bone - Bone: dyn_BackHair_M_03"!(ComponentPose=@n2, bAlphaBoolEnabled:bool=True, Translation="0,0,0", Rotation="-7,0,0", Scale="1,1,1") -> (Pose) {
    alphaInputType=Bool
    rotationMode=BMM_Additive
    rotationSpace=BCS_BoneSpace
    translationMode=BMM_Additive
    translationSpace=BCS_BoneSpace
    alphaBoolBlend {blendOption=Linear}
    alphaScaleBiasClamp {bInterpResult=true}
    boneToModify {boneName=dyn_BackHair_M_03}
    rotation {pitch=-7}
   }
   n4 = "Transform (Modify) Bone - Bone: dyn_BackHair_L_03"!(ComponentPose=@n3, bAlphaBoolEnabled:bool=True, Translation="0,0,0", Rotation="-5,0,0", Scale="1,1,1") -> (Pose) {
    alphaInputType=Bool
    rotationMode=BMM_Additive
    rotationSpace=BCS_BoneSpace
    translationMode=BMM_Additive
    translationSpace=BCS_BoneSpace
    alphaBoolBlend {blendOption=Linear}
    alphaScaleBiasClamp {bInterpResult=true}
    boneToModify {boneName=dyn_BackHair_L_03}
    rotation {pitch=-5}
   }
   n5 = "Transform (Modify) Bone - Bone: dyn_BackHair_R_03"!(ComponentPose=@n4, bAlphaBoolEnabled:bool=True, Translation="0,0,0", Rotation="0,0,0", Scale="1,1,1") -> (Pose) {
    alphaInputType=Bool
    rotationMode=BMM_Additive
    rotationSpace=BCS_BoneSpace
    translationMode=BMM_Additive
    translationSpace=BCS_BoneSpace
    alphaBoolBlend {blendOption=Linear}
    alphaScaleBiasClamp {bInterpResult=true}
    boneToModify {boneName=dyn_BackHair_R_03}
    rotation=<rot 0 0 0>
   }
   Output Pose(Result=Blend Poses by bool(
     BlendPose_0=Component To Local(ComponentPose=@n1)
     BlendPose_1=Component To Local(ComponentPose=@n5)
     bActiveValue=Get Is Backbling Equipped() -> (Is Backbling Equipped:bool)
    ) {blendType=Linear}) {layerGroup=None}
   n6 = "Get Face Mesh "() -> ("Face Mesh ":Skeletal Mesh Component Object Reference)
   n7 = Copy Pose From Mesh() {bUseAttachedParent=true}
   n8 = Copy Pose From Mesh(SourceMeshComponent="Get Body Mesh "() -> ("Body Mesh ":Skeletal Mesh Component Object Reference)) {bUseAttachedParent=true}
   Save cached pose 'Start_Cache'(Pose=Layered blend per bone(
     BasePose=Layered blend per bone(BasePose=@n7, BlendPoses_0=@n8)
     BlendPoses_0=Copy Pose From Mesh(SourceMeshComponent=@n6) {bUseAttachedParent=true}
    ) {layerSetup [{branchFilters [{boneName=FACIAL_R_ForeheadMid}]}]}) {cacheName=Start_Cache}
   n9 = Anim Dynamics - Bone: dyn_FrontBangA_L_01(ComponentPose=Local To Component(LocalPose=Use cached pose 'Start_Cache'()), Alpha=1, ExternalForce="0,0,0") {
    angularDampingOverride=1
    angularSpringConstant=20000
    bAngularSpring=true
    bOverrideAngularDamping=true
    bUseGravityOverride=true
    gravityScale=0.35
    simulationSpace=BoneRelative
    alphaBoolBlend {blendOption=Linear}
    boundBone {boneName=dyn_FrontBangA_L_01}
    componentLinearAccScale=<1 1 1>
    componentLinearVelScale=<1 1 1>
    gravityOverride {z=-980}
    physicsBodyDefinitions [{
      boundBone {boneName=dyn_FrontBangA_L_01}
      boxExtents=<20 30 30>
      constraintSetup {angularLimitsMax=<0 15 15>, angularLimitsMin=<0 -15 -15>, angularTarget=<1 0 0>}
      localJointOffset=<1 0 0>
     }]
    relativeSpaceBone {boneName=root}
   }
   n10 = Anim Dynamics - Bone: dyn_FrontBangB_L_01(ComponentPose=@n9, Alpha=1, ExternalForce="0,0,0") {
    angularDampingOverride=1
    angularSpringConstant=20000
    bAngularSpring=true
    bOverrideAngularDamping=true
    bUseGravityOverride=true
    gravityScale=0.35
    simulationSpace=BoneRelative
    alphaBoolBlend {blendOption=Linear}
    boundBone {boneName=dyn_FrontBangB_L_01}
    componentLinearAccScale=<1 1 1>
    componentLinearVelScale=<1 1 1>
    gravityOverride {z=-980}
    physicsBodyDefinitions [{
      boundBone {boneName=dyn_FrontBangB_L_01}
      boxExtents=<20 30 30>
      constraintSetup {angularLimitsMax=<0 15 15>, angularLimitsMin=<0 -15 -15>, angularTarget=<1 0 0>}
      localJointOffset=<1 0 0>
     }]
    relativeSpaceBone {boneName=root}
   }
   n11 = Anim Dynamics - Bone: dyn_frontBang_R_01(ComponentPose=@n10, Alpha=1, ExternalForce="0,0,0") {
    angularDampingOverride=1
    angularSpringConstant=20000
    bAngularSpring=true
    bOverrideAngularDamping=true
    bUseGravityOverride=true
    gravityScale=0.35
    simulationSpace=BoneRelative
    alphaBoolBlend {blendOption=Linear}
    boundBone {boneName=dyn_frontBang_R_01}
    componentLinearAccScale=<1 1 1>
    componentLinearVelScale=<1 1 1>
    gravityOverride {z=-980}
    physicsBodyDefinitions [{
      boundBone {boneName=dyn_frontBang_R_01}
      boxExtents=<20 30 30>
      constraintSetup {angularLimitsMax=<0 0 2.5>, angularLimitsMin=<0 -1.5 -2.5>, angularTarget=<1 0 0>}
      localJointOffset=<1 0 0>
     }]
    relativeSpaceBone {boneName=root}
   }
   n12 = Anim Dynamics - Bone: dyn_frontBang_R_02(ComponentPose=@n11, Alpha=1, ExternalForce="0,0,0") {
    angularDampingOverride=1
    angularSpringConstant=20000
    bAngularSpring=true
    bOverrideAngularDamping=true
    bUseGravityOverride=true
    gravityScale=0.35
    simulationSpace=BoneRelative
    alphaBoolBlend {blendOption=Linear}
    boundBone {boneName=dyn_frontBang_R_02}
    componentLinearAccScale=<1 1 1>
    componentLinearVelScale=<1 1 1>
    gravityOverride {z=-980}
    physicsBodyDefinitions [{
      boundBone {boneName=dyn_frontBang_R_02}
      boxExtents=<20 30 30>
      constraintSetup {angularLimitsMax=<0 0 2.5>, angularLimitsMin=<0 -1.5 -2.5>, angularTarget=<1 0 0>}
      localJointOffset=<1 0 0>
     }]
    relativeSpaceBone {boneName=dyn_frontBang_R_01}
   }
   n13 = Anim Dynamics - Bone: dyn_Hoop_L(ComponentPose=@n12, Alpha=1, ExternalForce="0,0,0") {
    angularDampingOverride=1
    angularSpringConstant=700
    bOverrideAngularDamping=true
    bUseGravityOverride=true
    gravityScale=0.35
    simulationSpace=BoneRelative
    alphaBoolBlend {blendOption=Linear}
    alphaScaleBiasClamp {bInterpResult=true, interpSpeedDecreasing=2.5, interpSpeedIncreasing=8}
    boundBone {boneName=dyn_Hoop_L}
    componentAppliedLinearAccClamp=<3000 3000 3000>
    componentLinearAccScale=<0.3 0.3 0.3>
    componentLinearVelScale=<0.3 0.3 0.3>
    gravityOverride {z=-980}
    physicsBodyDefinitions [{
      boundBone {boneName=dyn_Hoop_L}
      boxExtents=<5 8 5>
      constraintSetup {angularLimitsMax=<0 15 0>, angularLimitsMin=<0 -15 -20>, angularTarget=<1 0 0>}
      localJointOffset=<3 0 0>
     }]
    relativeSpaceBone {boneName=root}
   }
   n14 = Anim Dynamics - Bone: dyn_Hoop_R(ComponentPose=@n13, Alpha=1, ExternalForce="0,0,0") {
    angularDampingOverride=1
    angularSpringConstant=700
    bOverrideAngularDamping=true
    bUseGravityOverride=true
    gravityScale=0.35
    simulationSpace=BoneRelative
    alphaBoolBlend {blendOption=Linear}
    alphaScaleBiasClamp {bInterpResult=true, interpSpeedDecreasing=2.5, interpSpeedIncreasing=8}
    boundBone {boneName=dyn_Hoop_R}
    componentAppliedLinearAccClamp=<3000 3000 3000>
    componentLinearAccScale=<0.3 0.3 0.3>
    componentLinearVelScale=<0.3 0.3 0.3>
    gravityOverride {z=-980}
    physicsBodyDefinitions [{
      boundBone {boneName=dyn_Hoop_R}
      boxExtents=<5 8 5>
      constraintSetup {angularLimitsMax=<0 15 40>, angularLimitsMin=<0 -15 0>, angularTarget=<1 0 0>}
      localJointOffset=<3 0 0>
     }]
    relativeSpaceBone {boneName=root}
   }
   n15 = "Constraint - dyn_BackHair_R_04 (P)"(ComponentPose=@n14, Alpha=1, ConstraintWeights_0=1) {
    alphaBoolBlend {blendOption=Linear}
    boneToModify {boneName=dyn_BackHair_R_04}
    constraintSetup [{targetBone {boneName=dyn_BackHair_M_05}}]
    constraintWeights [1]
   }
   Save cached pose 'AnimDynm2Logic'(Pose=Component To Local(ComponentPose=@n15)) {cacheName=AnimDynm2Logic}
   n16 = Get Head Up() -> (Head Up:float)
   n17 = "Transform (Modify) Bone - Bone: base_BackHair_L"(
    ComponentPose=Local To Component(LocalPose=Use cached pose 'AnimDynm2Logic'())
    Alpha=@n16
    Translation="-0.158,1.030,0.658"
    Rotation="44.731,-0.113,-0.116"
    Scale="1,1,1"
   ) {
    rotationMode=BMM_Additive
    rotationSpace=BCS_BoneSpace
    translationMode=BMM_Additive
    translationSpace=BCS_BoneSpace
    alphaBoolBlend {blendOption=Linear}
    alphaScaleBiasClamp {bInterpResult=true}
    boneToModify {boneName=base_BackHair_L}
    rotation=<rot 44.731 -0.113 -0.116>
    translation=<-0.158 1.03 0.658>
   }
   n18 = "Transform (Modify) Bone - Bone: base_BackHair_M"(ComponentPose=@n17, Alpha=@n16, Translation="0.134,-0.541,1.089", Rotation="67.077,-0.616,0.163", Scale="1,1,1") {
    rotationMode=BMM_Additive
    rotationSpace=BCS_BoneSpace
    translationMode=BMM_Additive
    translationSpace=BCS_BoneSpace
    alphaBoolBlend {blendOption=Linear}
    alphaScaleBiasClamp {bInterpResult=true}
    boneToModify {boneName=base_BackHair_M}
    rotation=<rot 67.077 -0.616 0.163>
    translation=<0.134 -0.541 1.089>
   }
   n19 = "Transform (Modify) Bone - Bone: base_BackHair_R"(ComponentPose=@n18, Alpha=@n16, Translation="-1.789,0.281,-0.783", Rotation="44.731,0.046,-0.015", Scale="1,1,1") {
    rotationMode=BMM_Additive
    rotationSpace=BCS_BoneSpace
    translationMode=BMM_Additive
    translationSpace=BCS_BoneSpace
    alphaBoolBlend {blendOption=Linear}
    alphaScaleBiasClamp {bInterpResult=true}
    boneToModify {boneName=base_BackHair_R}
    rotation=<rot 44.731 0.046 -0.015>
    translation=<-1.789 0.281 -0.783>
   }
   n20 = "Transform (Modify) Bone - Bone: base_FrontHairA_R"(ComponentPose=@n19, Alpha=@n16, Translation="0,0,0", Rotation="-1.709,31.814,-5.334", Scale="1,1,1") {
    rotationMode=BMM_Additive
    rotationSpace=BCS_BoneSpace
    translationMode=BMM_Additive
    translationSpace=BCS_BoneSpace
    alphaBoolBlend {blendOption=Linear}
    alphaScaleBiasClamp {bInterpResult=true}
    boneToModify {boneName=base_FrontHairA_R}
    rotation=<rot -1.709 31.814 -5.334>
   }
   n21 = "Transform (Modify) Bone - Bone: base_FrontHairB_R"(ComponentPose=@n20, Alpha=@n16, Translation="0,0,0", Rotation="4.000,47.065,-4.800", Scale="1,1,1") {
    rotationMode=BMM_Additive
    rotationSpace=BCS_BoneSpace
    translationMode=BMM_Additive
    translationSpace=BCS_BoneSpace
    alphaBoolBlend {blendOption=Linear}
    alphaScaleBiasClamp {bInterpResult=true}
    boneToModify {boneName=base_FrontHairB_R}
    rotation=<rot 4 47.065 -4.8>
   }
   n22 = "Transform (Modify) Bone - Bone: base_BackHair_CorrectiveJoint_01"(ComponentPose=@n21, Alpha=1, Translation="8.494,0.551,2.329", Rotation="0,0,0", Scale="1.500,1.000,1.000") {
    rotationMode=BMM_Additive
    rotationSpace=BCS_BoneSpace
    scaleMode=BMM_Additive
    scaleSpace=BCS_BoneSpace
    translationMode=BMM_Additive
    translationSpace=BCS_BoneSpace
    alphaBoolBlend {blendOption=Linear}
    boneToModify {boneName=base_BackHair_CorrectiveJoint_01}
    scale {x=1.5}
    translation=<8.494 0.551 2.329>
   }
   n23 = Get is Float BackBling On() -> (is Float BackBling On:double)
   n24 = "Transform (Modify) Bone - Bone: base_BackHair_R"(ComponentPose=@n22, Alpha=@n23, Translation="0,0,0", Rotation="6.822,26.335,39.045", Scale="1,1,1") {
    rotationMode=BMM_Additive
    rotationSpace=BCS_BoneSpace
    translationMode=BMM_Additive
    translationSpace=BCS_BoneSpace
    alphaBoolBlend {blendOption=Linear}
    boneToModify {boneName=base_BackHair_R}
    rotation=<rot 6.822 26.335 39.045>
   }
   n25 = "Transform (Modify) Bone - Bone: base_BackHair_M"(ComponentPose=@n24, Alpha=@n23, Translation="0,0,0", Rotation="-0.000,-15.801,0.042", Scale="1,1,1") {
    rotationMode=BMM_Additive
    rotationSpace=BCS_BoneSpace
    translationMode=BMM_Additive
    translationSpace=BCS_BoneSpace
    alphaBoolBlend {blendOption=Linear}
    boneToModify {boneName=base_BackHair_M}
    rotation=<rot 0 -15.801 0.042>
   }
   n26 = "Transform (Modify) Bone - Bone: base_BackHair_L"(ComponentPose=@n25, Alpha=@n23, Translation="0,0,0", Rotation="11.081,24.683,-24.043", Scale="1,1,1") {
    rotationMode=BMM_Additive
    rotationSpace=BCS_BoneSpace
    translationMode=BMM_Additive
    translationSpace=BCS_BoneSpace
    alphaBoolBlend {blendOption=Linear}
    boneToModify {boneName=base_BackHair_L}
    rotation=<rot 11.081 24.683 -24.043>
   }
   n27 = Get is Aiming Up() -> (is Aiming Up:double)
   n28 = "Transform (Modify) Bone - Bone: base_BackHair_R"(ComponentPose=@n26, Alpha=@n27, Translation="-0.819,-0.124,-0.247", Rotation="18.015,-20.819,-10.114", Scale="1,1,1") {
    rotationMode=BMM_Additive
    rotationSpace=BCS_BoneSpace
    translationMode=BMM_Additive
    translationSpace=BCS_BoneSpace
    alphaBoolBlend {blendOption=Linear}
    alphaScaleBiasClamp {bInterpResult=true}
    boneToModify {boneName=base_BackHair_R}
    rotation=<rot 18.015 -20.819 -10.114>
    translation=<-0.819 -0.124 -0.247>
   }
   n29 = "Transform (Modify) Bone - Bone: base_BackHair_M"(ComponentPose=@n28, Alpha=@n27, Translation="0,0,0", Rotation="-0.615,-0.013,0.009", Scale="1,1,1") {
    rotationMode=BMM_Additive
    rotationSpace=BCS_BoneSpace
    translationMode=BMM_Additive
    translationSpace=BCS_BoneSpace
    alphaBoolBlend {blendOption=Linear}
    alphaScaleBiasClamp {bInterpResult=true}
    boneToModify {boneName=base_BackHair_M}
    rotation=<rot -0.6151 -0.013 0.009>
   }
   n30 = "Transform (Modify) Bone - Bone: base_BackHair_R"(
    ComponentPose=@n29
    Alpha=Get is Skydive Float() -> (is Skydive Float:double)
    Translation="0,0,0"
    Rotation="6.822,-26.335,-0.024"
    Scale="1,1,1"
   ) {
    rotationMode=BMM_Additive
    rotationSpace=BCS_BoneSpace
    translationMode=BMM_Additive
    translationSpace=BCS_BoneSpace
    alphaBoolBlend {blendOption=Linear}
    boneToModify {boneName=base_BackHair_R}
    rotation=<rot 6.822 -26.335 -0.024>
   }
   Save cached pose 'Logic2RigidBody'(Pose=Component To Local(ComponentPose=@n30)) {cacheName=Logic2RigidBody}
  ]}
 EventGraph {nodes [
   Event BlueprintUpdateAnimation()
   > n1 = Sequence()
   > n2 = Gravity Override State Function DEPRECATED(CharacterPartAnimInstance=Self-Reference(), GameTimeInSeconds=GetGameTimeInSeconds(), GravityOverrideParams=Get Gravity Override Params())
   > Set Gravity Override(Gravity Override=@n2)
   > Set is Skydive Float(is Skydive Float:double="To Float (Boolean)"(InBool=Get bIsSkydiveFloating())) -> (Output_Get:double)
   > n3 = Branch(Condition=Get bIsTargeting())
   > n4 = Set is Aiming Up(is Aiming Up:double=Get Head Up()) -> (Output_Get:double)
   n5 = Branch(execute=[@n6.then, @n4.then], Condition=Get Is Backbling Equipped())
   > Set is Float BackBling On(is Float BackBling On:double=Get is Skydive Float()) -> (Output_Get:double)
   Set is Float BackBling On(execute=@n5.else, is Float BackBling On=0)
   n6 = Set is Aiming Up(execute=@n3.else, is Aiming Up=0)
   Set Head Up(
    execute=@n1.then_1
    Head Up:float=GetAlphaFromReferenceTransformDelta(SkeletalMeshComponent=Get Skel Mesh(), BoneName=spine_04, BaseName=head, InputRange=(X=15,Y=60), TwistAxis="0,0,1")
   ) -> (Output_Get:float)
   > Set Head Down(Head Down:float=GetAlphaFromReferenceTransformDelta(
     SkeletalMeshComponent=Get Skel Mesh()
     BoneName=spine_04
     BaseName="head "
     InputRange=(X=0,Y=-60)
     OutputRange=(X=0,Y=-1)
     TwistAxis="0,0,1"
    )) -> (Output_Get:float)
   n7 = CheckForBackbling()
   > n8 = Set Is Wearing Backpack(Is Wearing Backpack:bool=AND Boolean!(A=@n7, B=NOT Boolean(A=@n7.bIsBackblingACape), C:bool=NOT Boolean(A=@n7.bIsBackblingLong)) -> (ReturnValue)) -> (Output_Get:bool)
   > n9 = Set Is Wearing Cape(Is Wearing Cape=OR Boolean(A=@n7.bIsBackblingACape, B=@n7.bIsBackblingLong))
   > Set Is Backbling Equipped(Is Backbling Equipped:bool=OR Boolean(A=@n9, B=@n8)) -> (Output_Get:bool)
   Event BlueprintInitializeAnimation()
   > Set Skel Mesh(Skel Mesh=GetOwnerSkeletalMeshComponent())
   Event OnCharacterCustomizationLoaded()
   > n10 = Cast To PlayerPawn_Athena(Object=TryGetPawnOwner())
   > "Set Body Mesh "("Body Mesh ":Skeletal Mesh Component Object Reference=GetSkeletalMeshForPartType(self=@n10)) -> (Output_Get:Skeletal Mesh Component Object Reference)
   > "Set Face Mesh "("Face Mesh ":Skeletal Mesh Component Object Reference=GetSkeletalMeshForPartType(self=@n10, PartType=Head)) -> (Output_Get:Skeletal Mesh Component Object Reference)
  ]}
]
variables [
 "Body Mesh " {type=SkeletalMeshComponent}
 "Face Mesh " {type=SkeletalMeshComponent}
 Gravity Override {type=Vector}
 Gravity Override Params {type=GravityOverrideParamsStruct}
 Head Down {type=number}
 Head Up {type=number}
 is Aiming Up {type=number}
 Is Backbling Equipped {type=boolean}
 is Float BackBling On {type=number}
 is Skydive Float {type=number}
 Is Wearing Backpack {type=boolean}
 Is Wearing Cape {type=boolean}
 Skel Mesh {type=SkeletalMeshComponent}
]
